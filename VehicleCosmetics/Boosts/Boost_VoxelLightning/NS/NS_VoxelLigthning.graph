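# Niagara system for the Voxel Lightning vehicle boost cosmetic, emitting trails and particles from the car's boost ports.
# Four emitters layer ribbons, two-group mesh/sprite debris, a rapidly looping flash sprite, and axis-oriented mesh/sprite particles driven by material parameters.
# Each places particles per filtered skeletal socket, shifted by DTCarOffset; ribbons and debris inherit source movement with a fade.
# Painted-variant overrides recolor every layer, and bIsPainted switches the ribbon's fade curves.
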NiagaraSystem /VehicleCosmetics/Boosts/Boost_VoxelLightning/NS/NS_VoxelLigthning
dependencies {
 usedDataInterfaces [
  &NiagaraDataInterfaceCurve
  &NiagaraDataInterfaceVectorCurve
  &NiagaraDataInterfaceArrayFloat
  &NiagaraDataInterfaceArrayColor
  &NiagaraDataInterfaceArrayFloat3
  &NiagaraDataInterfaceSkeletalMesh
 ]
 usedDynamicInputs []
 usedModules [
  &self:SystemSpawnScript.SystemScriptSource.SystemScriptGraph.NiagaraNodeAssignment_0.SetVariables_#1
  &SystemState
 ]
}
emitters {
 NE_VC_Boost_Ribbon_Parent001 {
  emitterUpdateScript [
   &EmitterState()
   &VC_FilteredSocketCount()
   &VC_SpawnRibbon_v1(Number of Ribbons=$Emitter.PortCount, Spawn Time=-1, SpawnRate=30)
  ]
  particleSpawnScript [
   &VC_PaintedVariantsOverride(
    Black=<rgba 0.2 0.2 0.2 0.9>
    BurntSienna=<rgba 1.5 0.3312 0.225 1>
    Cobalt=<rgba 0.3 0.75 3 1>
    Crimson=<rgba 3.2 0.384 0.4309 0.75>
    DefaultColor=<rgba 0.775 2.5 0.2 1>
    ForestGreen=<rgba 0.35 3.5 0.35 1>
    Grey=<rgba 0.4 0.4 0.4 1>
    Lime=<rgba 1.67 3 0.15 1>
    Orange=<rgba 2.5 0.38 0.1 1>
    Pink=<rgba 3 0.6 2.4 1>
    Purple=<rgba 0.56 0.14 3.5 1>
    Saffron=<rgba 3 2.15 0.45 1>
    SkyBlue=<rgba 0.255 1.2925 1.5 1>
    TitaniumWhite=<rgba 2 2 2 1>
   )
   &VC_InitializeRibbon_v1(
    Color=$Particles.Color
    Custom Offset=true
    Enable Skeletal Position=true
    Lifetime=1
    Local Vector Offset=&AddVectorToPosition(Position=<-65 0 0>, Vector=$User.DTCarOffset)
    Offset Amount=1
    Ribbon Width=8.5
   )
   &AddVelocity(Velocity=<-1000 0 0>) {module=AddVelocity001}
  ]
  particleUpdateScript [
   &ParticleState()
   &ScaleColor(
    Scale Alpha=&MakeCustomFloatFromBool(
     Bool=$User.bIsPainted
     True Float=&FloatFromCurve(
      FloatCurve=DataInterface() {propertyValues {Curve {keys [
          {arriveTangent=-1.7675, interpMode=RCIM_Cubic, leaveTangent=-1.7675, leaveTangentWeight=0.3501, tangentMode=RCTM_User, time=0, value=1}
          {arriveTangent=-0.3794, interpMode=RCIM_Cubic, leaveTangent=-0.3794, tangentMode=RCTM_User, time=0.7, value=0.1}
          <at 1.0035 0 RCIM_Cubic>
         ]}}}
      Scale Curve=0.4
     )
     False Float=&FloatFromCurve(FloatCurve=DataInterface() {propertyValues {Curve {keys [<at 0 1 RCIM_Cubic>, {arriveTangentWeight=0.3004, interpMode=RCIM_Cubic, tangentMode=RCTM_User, time=0.9, value=0}]}}})
    )
    Scale RGB=&MakeCustomVectorFromBool(
     Bool=$User.bIsPainted
     True Vector=&VectorFromFloat(Value=&FloatFromCurve(FloatCurve=DataInterface() {propertyValues {Curve {keys [
           <at 0.0005 0.9 RCIM_Cubic>
           {arriveTangent=-0.8054, interpMode=RCIM_Cubic, leaveTangent=-0.8054, tangentMode=RCTM_User, time=0.35, value=0.95}
           {arriveTangentWeight=0.15, interpMode=RCIM_Cubic, tangentMode=RCTM_User, time=0.75, value=0.25}
          ]}}}))
     False Vector=&VectorFromCurve(VectorCurve=DataInterface() {propertyValues {XCurve {keys [<at 0.3 1 RCIM_Cubic>,<at 0.8 0 RCIM_Cubic>]}, YCurve {keys [<at 0 1>,<at 1 1>]}, ZCurve {keys [<at 0 1>,<at 1 0>]}}})
    )
    ScaleA=true
    ScaleRGB=true
   )
   &ScaleRibbonWidth(Ribbon Width Scale=&FloatFromCurve(
     FloatCurve=DataInterface() {propertyValues {Curve {keys [{tangentMode=RCTM_User, time=0, value=1.25}, {tangentMode=RCTM_User, time=1, value=2}]}}}
     Scale Curve=&SelectFloatFromArray(Direct Array Index=$Particles.SpawnGroup, Float Selection Array=DataInterface() {propertyValues {FloatData [3.5,3.5]}})
    ))
   &InheritSourceMovement(
    Applied Position Scale=&VectorFromFloat(Value=&FloatFromCurve(FloatCurve=DataInterface() {propertyValues {Curve {keys [
          {interpMode=RCIM_Cubic, tangentMode=RCTM_Break, time=0, value=1}
          {arriveTangent=-2.9025, interpMode=RCIM_Cubic, tangentMode=RCTM_Break, time=1, value=0}
         ]}}}))
    ApplyForce=false
    ApplyPosition=true
    ApplyPositionFalloff=false
   )
   &SolveForcesAndVelocity(Clamp Velocity=false, Limit Acceleration=false)
   &VC_UpdateRibbon_v1(Enable Skeletal Position=true)
  ]
  rendererClasses [&NiagaraRibbonRendererProperties]
  renderers [{rendererClass=&NiagaraRibbonRendererProperties}]
 }
 NE_VC_Boost_Sprite_Parent001 {
  emitterUpdateScript [
   &EmitterState()
   &SpawnRate(SpawnGroup=1, SpawnRate=20, Use Spawn Probability=false) {module=SpawnRate001}
   &VC_FilteredSocketCount()
   &SpawnRate(SpawnRate=20, Use Spawn Probability=false)
  ]
  particleSpawnScript [
   /SetVariables(Particles.VisibilityTag=$Emitter.SpawnGroup) {isSetParameters=true}
   &VC_PaintedVariantsOverride(
    Black=<rgba 0.04 0.04 0.04 1>
    BurntSienna=&RandomRangeLinearColor(Minimum=<rgba 1 0.13 0.1 1>, Maximum=<rgba 0.9 0.144 0.09 1>)
    Cobalt=&RandomRangeLinearColor(Minimum=<rgba 0 0.075 1.5 1>, Maximum=<rgba 0.05 0.0975 1 1>)
    Crimson=&RandomRangeLinearColor(Minimum=<rgba 0.5 0 0.0136 1>, Maximum=<rgba 0.5 0.05 0.0622 1>)
    DefaultColor=&SelectLinearColorFromArray(
     Array Sampling Mode=::NewEnumerator0
     Direct Array Index=?
     Color Selection Array=DataInterface() {propertyValues {ColorData [<rgba 0.4375 0.025 2.5 1>, <rgba 0.3733 0.2 1 1>]}}
     Random Seed=0
    )
    ForestGreen=<rgba 0 0.25 0.0126 1>
    Grey=<rgba 0.2 0.2 0.2 1>
    Lime=<rgba 0.285 0.6 0.06 1>
    Orange=<rgba 0.8 0.0933 0 1>
    Pink=<rgba 0.7 0.105 0.7 1>
    Purple=<rgba 0.11 0 1.1 1>
    R1 Black:LinearColor=<rgba 0.02 0.02 0.02 1>
    R1 BurntSienna:LinearColor=<rgba 0.15 0.0218 0.015 1>
    R1 Cobalt:LinearColor=<rgba 0 0.0092 0.5 1>
    R1 Crimson:LinearColor=<rgba 0.4635 0.0093 0.0216 1>
    R1 DefaultColor=<rgba 0 1 0 0.2>
    R1 ForestGreen:LinearColor=<rgba 0 0.5 0 0.5>
    R1 Lime:LinearColor=<rgba 0.05 0.1 0 1>
    R1 Orange:LinearColor=<rgba 0.672 0.056 0 1>
    R1 Pink:LinearColor=<rgba 0.2 0.024 0.136 1>
    R1 Purple:LinearColor=<rgba 0.05 0 0.3 1>
    R1 Saffron:LinearColor=<rgba 0.5 0.3833 0 1>
    Saffron=<rgba 1.1 0.8433 0 1>
    SkyBlue=<rgba 0 0.5833 1.4 1>
    TitaniumWhite=<rgba 1.1 1.1 1.1 1>
    Use R1 Painted Overrides=true
   )
   &InitializeParticle(
    Color=$Particles.Color
    Lifetime=&MergeHelper_FloatSwitch(
     0=0.6
     1=&RandomRangeFloat(Minimum=0.35, Maximum=0.5)
     2=?
     3=?
     Index=&SpawnGroup()
     Num=::NewEnumerator0
    )
    Mesh Scale Mode=::NewEnumerator1
    Mesh Uniform Scale Max=0.03
    Mesh Uniform Scale Min=0.03
    Sprite Rotation Mode=::NewEnumerator2
    Sprite Size Mode=::NewEnumerator3
    Uniform Sprite Size=&RandomRangeFloat(Minimum=1, Maximum=1.5)
    UsePositionOffset=false
    Write Material Random=true
   )
   &PartitionParticles(Number Of Partitions=$Emitter.PortCount, Particles Per Partition=2, Partition Mode:ENiagara_PartitionMode=::NewEnumerator1)
   &SkeletalMeshLocation(
    Bone / Socket Index=$Particles.PartitionParticles.Partition
    Bone / Socket Sampling Mode=::NewEnumerator5
    MaskSpawnGroup=false
    Mesh Sampling Type=::NewEnumerator6
    Sampled Position Offset=&AddVectorToPosition(
     Position=&MergeHelper_VectorSwitch(0=<-50 0 0>, 1=<-85 0 0>, 2=?, 3=?, Index=&SpawnGroup(), Num=::NewEnumerator0)
     Vector=$User.DTCarOffset
    )
    Skeletal Mesh=$System.SkeletalMesh
   )
   &AddVelocity(Velocity=&MergeHelper_VectorSwitch(
     0=<-850 0 0>
     1=&RandomRangeVector(Minimum=<-200 0 0>, Maximum=<-250 0 0>)
     2=?
     3=?
     Index=&SpawnGroup()
     Num=::NewEnumerator0
    ))
   &ShapeLocation(
    MaskSpawnGroup=false
    Override Seed=false
    Sphere Radius=&MergeHelper_FloatSwitch(0=20, 1=30, 2=?, 3=?, Index=&SpawnGroup(), Num=::NewEnumerator0)
    UseOwnerScale=true
   )
  ]
  particleUpdateScript [
   &ParticleState()
   /SetVariables(Particles.R1scalecolor:LinearColor=&VC_DynamicScaleColor(
     Color to Scale:LinearColor=$Particles.R1 Color
     RGB Scale:Vector3f=&VectorFromFloat(Value=&FloatFromCurve(FloatCurve=DataInterface() {propertyValues {Curve {keys [<at 0 15>,<at 1 10>]}}}))
     Alpha Scale:NiagaraFloat=1
    )) {isSetParameters=true}
   &ScaleSpriteSize(
    Uniform Curve Scale=3
    Uniform Curve Sprite Scale=DataInterface() {propertyValues {Curve {keys [{tangentMode=RCTM_User, time=0, value=1.5}, {tangentMode=RCTM_User, time=1, value=0}]}}}
   )
   &InheritSourceMovement(
    Applied Position Scale=&VectorFromFloat(Value=&FloatFromCurve(FloatCurve=DataInterface() {propertyValues {Curve {keys [
          {interpMode=RCIM_Cubic, tangentMode=RCTM_Break, time=0, value=1}
          {arriveTangent=-2.9025, interpMode=RCIM_Cubic, tangentMode=RCTM_Break, time=1, value=0}
         ]}}}))
    ApplyForce=false
    ApplyPosition=true
    ApplyPositionFalloff=false
   )
   &SolveForcesAndVelocity(Clamp Velocity=false, Limit Acceleration=false)
   &DynamicMaterialParameters(
    Index 0 Param 1=1
    Module.Index0Float0Write=true
    Module.Index0Float1Write=true
    Module.Index0Float2Write=true
    Module.Index0Float3Write=true
   )
   &ScaleColor(Scale RGB=&VectorFromFloat(Value=0.4), ScaleA=true, ScaleRGB=true)
  ]
  rendererClasses [&NiagaraMeshRendererProperties, &NiagaraSpriteRendererProperties]
  renderers [
   {rendererClass=&NiagaraMeshRendererProperties, rendererIndex=0}
   {rendererClass=&NiagaraSpriteRendererProperties, rendererIndex=1}
   {rendererClass=&NiagaraSpriteRendererProperties, rendererIndex=2}
  ]
 }
 NE_VC_Boost_Sprite_Parent003 {
  emitterUpdateScript [
   &EmitterState(Inactive Response=::NewEnumerator1, Life Cycle Mode=::NewEnumerator1, Loop Duration=0.075, UseLoopDelay=false)
   &SpawnBurst_Instantaneous(Use Spawn Probability=false, UseLoopCountLimit=false)
   &VC_FilteredSocketCount()
   &SpawnRate(SpawnRate=4, Use Spawn Probability=false) {enabled=false}
  ]
  particleSpawnScript [
   &VC_PaintedVariantsOverride(
    Black=<rgba 0.5 0.5 0.5 0.5>
    BurntSienna=<rgba 2 0.6175 0.2465 0.5>
    Cobalt=<rgba 0.4 0.6667 2 2>
    Crimson=<rgba 2 0.2 0.2488 2>
    DefaultColor=<rgba 0.6 1 0.5 1.5>
    ForestGreen=<rgba 0 1.5 0.0758 2>
    Lime=<rgba 0.825 1.5 0.15 1.5>
    Orange=<rgba 2 0.7627 0.1 1>
    Pink=<rgba 1.5 0.45 1.1182 2>
    Purple=<rgba 0.65 0.2 2 2>
    Saffron=<rgba 1.5 1.22 0.3 2>
    SkyBlue=<rgba 0.3 1.0985 1.5 2>
    TitaniumWhite=<rgba 0.75 0.75 0.75 2>
   )
   &InitializeParticle(
    Color=$Particles.Color
    Lifetime=0.08
    Sprite Size Mode=::NewEnumerator3
    Uniform Sprite Size=&MakeCustomFloatFromBool(Bool=$User.bIsPainted, True Float=2.7, False Float=2.7)
    UsePositionOffset=false
    Write Material Random=true
   )
   &PartitionParticles(Number Of Partitions=$Emitter.PortCount, Particles Per Partition=1, Partition Mode:ENiagara_PartitionMode=::NewEnumerator1)
   &SkeletalMeshLocation(
    Bone / Socket Index=$Particles.PartitionParticles.Partition
    Bone / Socket Sampling Mode=::NewEnumerator5
    MaskSpawnGroup=false
    Mesh Sampling Type=::NewEnumerator6
    Sampled Position Offset=&AddVectorToPosition(Position=<-70 0 0>, Vector=$User.DTCarOffset)
    Skeletal Mesh=$System.SkeletalMesh
   )
   &AddVelocity(Velocity=&RandomRangeVector(Minimum=<-100 0 0>, Maximum=<-65 0 0>)) {enabled=false}
  ]
  particleUpdateScript [
   &ParticleState()
   &ScaleColor(
    Scale Alpha=2
    Scale RGB=&VectorFromFloat(Value=&FloatFromCurve(FloatCurve=DataInterface() {propertyValues {Curve {keys [<at 0 0>,<at 0.0957 0.9043>,<at 1 0>]}}}, Scale Curve=10))
    ScaleA=true
    ScaleRGB=true
   )
   &ScaleSpriteSize(
    Scale Sprite Size Mode=::NewEnumerator0
    Uniform Scale Factor=&FloatFromCurve(FloatCurve=DataInterface() {propertyValues {Curve {keys [<at 0 0>,<at 0.0994 0.9006>,<at 1 0>]}}}, Scale Curve=45)
   )
   &SolveForcesAndVelocity(Clamp Velocity=false, Limit Acceleration=false)
  ]
  rendererClasses [&NiagaraSpriteRendererProperties]
  renderers [{rendererClass=&NiagaraSpriteRendererProperties}]
 }
 Nuevo {
  emitterUpdateScript [
   &EmitterState()
   &VC_FilteredSocketCount()
   &SpawnRate(SpawnRate=90, Use Spawn Probability=false)
  ]
  particleSpawnScript [
   &VC_PaintedVariantsOverride(
    Black=&RandomRangeLinearColor(Minimum=<rgba 0.04 0.04 0.04 1>, Maximum=<rgba 0.01 0.01 0.01 1>)
    BurntSienna=&RandomRangeLinearColor(Minimum=<rgba 0.25 0.0196 0.0075 1>, Maximum=<rgba 0.45 0.087 0.054 1>)
    Cobalt=&RandomRangeLinearColor(Minimum=<rgba 0.225 0.395 1.5 1>, Maximum=<rgba 0.06 0.117 1.2 1>)
    Crimson=&RandomRangeLinearColor(Minimum=<rgba 0.6 0 0.02 1>, Maximum=<rgba 1.2 0.156 0.2082 1>)
    DefaultColor=&SelectLinearColorFromArray(
     Array Sampling Mode=::NewEnumerator0
     Direct Array Index=?
     Color Selection Array=DataInterface() {propertyValues {ColorData [<rgba 0.4 0.25 1 1>, <rgba 0.36 0.252 0.9 1>, <rgba 0.4333 0.15 1 1>, <rgba 0.32 0.15 0.75 1>]}}
     Random Seed=0
    )
    ForestGreen=&RandomRangeLinearColor(Minimum=<rgba 0 0.3 0.0152 1>, Maximum=<rgba 0 0.95 0 1>)
    Grey=&RandomRangeLinearColor(Minimum=<rgba 0.2 0.2 0.2 1>, Maximum=<rgba 0.74 0.74 0.74 1>)
    Lime=&RandomRangeLinearColor(Minimum=<rgba 0.325 0.65 0 1>, Maximum=<rgba 0.7475 1.3 0.195 1>)
    Orange=&RandomRangeLinearColor(Minimum=<rgba 1 0.0833 0 1>, Maximum=<rgba 1 0.2 0 1>)
    Pink=&RandomRangeLinearColor(Minimum=<rgba 0.38 0.038 0.38 1>, Maximum=<rgba 0.9317 0.18 1 1>)
    Purple=&RandomRangeLinearColor(Minimum=<rgba 0.205 0.1 1 1>, Maximum=<rgba 0.33 0.15 1.5 1>)
    Saffron=&RandomRangeLinearColor(Minimum=<rgba 0.5 0.3333 0 1>, Maximum=<rgba 1 0.7308 0.05 1>)
    SkyBlue=&RandomRangeLinearColor(Minimum=<rgba 0 0.16 0.32 1>, Maximum=<rgba 0.05 0.6821 1 1>)
    TitaniumWhite=&RandomRangeLinearColor(Minimum=<rgba 0.55 0.55 0.55 1>, Maximum=<rgba 1 1 1 1>)
   )
   &InitializeParticle(
    Color=$Particles.Color
    Lifetime Max=0.35
    Lifetime Min=0.3
    Lifetime Mode=::NewEnumerator1
    Mesh Scale Mode=::NewEnumerator3
    Mesh Uniform Scale=0.003
    UsePositionOffset=false
    Write Material Random=true
   )
   &PartitionParticles(Number Of Partitions=$Emitter.PortCount)
   &SkeletalMeshLocation(
    Bone / Socket Index=$Particles.PartitionParticles.Partition
    Bone / Socket Sampling Mode=::NewEnumerator5
    MaskSpawnGroup=false
    Mesh Sampling Type=::NewEnumerator6
    Sampled Position Offset=&AddVectorToPosition(Position=<-75 0 0>, Vector=$User.DTCarOffset)
    Skeletal Mesh=$System.SkeletalMesh
   )
   &AddVelocity(Velocity=&SelectVectorFromArray(
     Array Sampling Mode=::NewEnumerator0
     Direct Array Index=?
     Vector Selection Array=DataInterface() {propertyValues {FloatData [<0 0 0>, <0 0 0>]}}
     Random Seed=0
    )) {enabled=false}
   &InitialMeshOrientation(
    Mesh Axis To Orient=&SelectVectorFromArray(
     Array Sampling Mode=::NewEnumerator0
     Direct Array Index=?
     Vector Selection Array=DataInterface() {propertyValues {FloatData [<1 0 0>, <0 1 0>, <0 0 1>]}}
     Random Seed=0
    )
    Mesh Orientation Mode=::NewEnumerator2
    Orientation Vector=&SelectVectorFromArray(
     Array Sampling Mode=::NewEnumerator0
     Direct Array Index=?
     Vector Selection Array=DataInterface() {propertyValues {FloatData [<1 0 0>, <0 1 0>, <0 0 1>]}}
     Random Seed=0
    )
    Use Rotation Vector=false
   )
   &ShapeLocation(MaskSpawnGroup=false, Override Seed=false, Sphere Radius=15, UseOwnerScale=true)
  ]
  particleUpdateScript [
   &ParticleState()
   &ScaleColor(
    Scale Alpha=&FloatFromCurve(FloatCurve=DataInterface() {propertyValues {Curve {keys [<at 0 0>,<at 0.7 3>,<at 1 10>]}}}, Scale Curve=9)
    Scale RGB=&VectorFromFloat(Value=&FloatFromCurve(
      FloatCurve=DataInterface() {propertyValues {Curve {defaultValue=1, keys [{tangentMode=RCTM_User, time=0, value=0.6}, <at 0.4965 0.8>, {tangentMode=RCTM_User, time=1, value=1.2}]}}}
      Scale Curve=0.65
     ))
    ScaleA=true
    ScaleRGB=true
   )
   &DynamicMaterialParameters(
    Index 0 Param 1=&FloatFromCurve(FloatCurve=DataInterface() {propertyValues {Curve {keys [<at 0.35 0.95>,<at 1 0>]}}}, Scale Curve=4)
    Index 0 Param 2=&FloatFromCurve(
     FloatCurve=DataInterface() {propertyValues {Curve {keys [
         {arriveTangent=-3.0454, interpMode=RCIM_Cubic, leaveTangent=-3.0454, tangentMode=RCTM_User, time=0, value=5}
         {arriveTangent=4.0872, interpMode=RCIM_Cubic, leaveTangent=4.0872, tangentMode=RCTM_User, time=1, value=45}
        ]}}}
     Scale Curve=&RandomRangeFloat(Minimum=20, Maximum=10)
    )
    Module.Index0Float0Write=true
    Module.Index0Float1Write=true
    Module.Index0Float2Write=true
    Module.Index0Float3Write=true
   )
   &SolveForcesAndVelocity(Clamp Velocity=false, Limit Acceleration=false)
  ]
  rendererClasses [&NiagaraMeshRendererProperties, &NiagaraSpriteRendererProperties]
  renderers [
   {rendererClass=&NiagaraMeshRendererProperties, rendererIndex=0}
   {rendererClass=&NiagaraSpriteRendererProperties, rendererIndex=1}
  ]
 }
}
summary {
 systemName=NS_VoxelLigthning
 emitters [
  {emitterName=NE_VC_Boost_Sprite_Parent003, rendererClasses [&NiagaraSpriteRendererProperties]}
  {emitterName=NE_VC_Boost_Ribbon_Parent001, rendererClasses [&NiagaraRibbonRendererProperties]}
  {emitterName=Nuevo, rendererClasses [&NiagaraMeshRendererProperties, &NiagaraSpriteRendererProperties]}
  {
   emitterName=NE_VC_Boost_Sprite_Parent001
   rendererClasses [&NiagaraMeshRendererProperties, &NiagaraSpriteRendererProperties]
  }
 ]
 userVariables [
  User.bIsFrontendPreview {type=NiagaraBool}
  User.bIsPainted {type=NiagaraBool}
  User.ColorVariantID {type=NiagaraInt32}
  User.DTCarOffset {type=Vector3f, description="", defaultValue {struct=Vector3f, value=<0 0 0>}}
  User.PaintedColor {type=LinearColor, description="", defaultValue {struct=LinearColor, value=<rgba 1 1 1 1>}}
  User.PaintedScalar {type=NiagaraFloat, defaultValue=1, description=""}
  User.PortCount {type=NiagaraInt32, defaultValue=2}
 ]
}
